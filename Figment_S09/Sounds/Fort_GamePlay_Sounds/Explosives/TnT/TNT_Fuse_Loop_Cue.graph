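# Sound cue for the looping fuse sound of TNT explosives.
# A single looping wave player feeds TNT_Fuse_Loop01 straight to the output at reduced volume, routed through the Explosives sound class.
# The cue overrides attenuation with natural-sound falloff and no stereo spread.
# Its concurrency override uses a prevent-new rule, so extra instances beyond the limit do not start.
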SoundCue /Figment_S09/Sounds/Fort_GamePlay_Sounds/Explosives/TnT/TNT_Fuse_Loop_Cue
root=WavePlayer_1
nodes [{id=WavePlayer_1, properties {bLooping=true, soundWaveAssetPtr=/Game/Sounds/Fort_GamePlay_Sounds/Explosives/TnT/TNT_Fuse_Loop01}}]
properties {
 bOverrideAttenuation=true
 bOverrideConcurrency=true
 maxDistance=2400
 soundClassObject=/Game/Sounds/FortSoundClassesAndModes/Fort_Explosives
 volumeMultiplier=0.2
 attenuationOverrides {distanceAlgorithm=NaturalSound, falloffDistance=2000, stereoSpread=0}
 concurrencyOverrides {resolutionRule=PreventNew}
}
